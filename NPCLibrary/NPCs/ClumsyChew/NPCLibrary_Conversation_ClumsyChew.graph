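# Conversation database for the ClumsyChew NPC, defining its opening greeting and top-level conversation branches.
# A single entry point feeds a Speech node that has a default line plus prioritized contextual lines gated on first-ever meeting, per-match conversation count, or wearing a specific PinkBear outfit.
# The speech then branches through link nodes, resolved by entry tag, to the shared NPCLibrary quests, basic services and ClumsyChew's own dialogue entry.
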ConversationDatabase /NPCLibrary/NPCs/ClumsyChew/NPCLibrary_Conversation_ClumsyChew
entry_points [{tag=Conversation.Entry.ClumsyChew, destinations [#1]}]
nodes [
 ConversationEntryPointNode_7 {
  id=#1
  class=ConversationEntryPointNode
  outputs [#2]
  properties {entryTag {tagName=Conversation.Entry.ClumsyChew}}
 }
 Speech_C_4 {
  id=#2
  class=Speech_C
  outputs [#3, #4, #5]
  properties {
   nodeName=""
   generalConfig {
    defaultMessage=These bugs won't know what squished 'em.
    contextualMessages [
     {
      message=Welcome to the fight. Hope you're ready for it.
      contextRequirements [{requirement {class=FortControllerRequirement_IsFirstEverConversationWithNPC}}]
      priority {curve {rowName=Tandem.Speech.Priority.FirstEverInteraction}}
     }
     {
      message=These bugs won't know what squished 'em.
      contextRequirements [{requirement {class=FortControllerRequirement_CheckNPCMatchConversationCount, properties {nPCUniqueIdTag {tagName=AISpawnerData.Type.Tandem.ClumsyChew}}}}]
      priority {curve {rowName=None}}
     }
     {
      message=It's about to get grisly.
      contextRequirements [{requirement {class=FortControllerRequirement_CheckNPCMatchConversationCount, properties {nPCUniqueIdTag {tagName=AISpawnerData.Type.Tandem.ClumsyChew}}}}]
      priority {curve {rowName=None}}
     }
     {
      message=Please... no more hugs...
      contextRequirements [{requirement {
         class=FortControllerRequirement_HasCID
         properties {allowedCIds [/Game/Athena/Items/Cosmetics/Characters/CID_069_Athena_Commando_F_PinkBear]}
        }}]
      priority {value=2, curve {curveTable=null, rowName=None}}
     }
    ]
   }
  }
 }
 ConversationLinkNode_2 {
  id=#5
  class=ConversationLinkNode
  properties {remoteEntryTag {tagName=Conversation.Entry.ClumsyChew.Dialogue}}
 }
 ConversationLinkNode_0 {id=#4, class=ConversationLinkNode, properties {remoteEntryTag {tagName=Conversation.Entry.BasicServices}}}
 ConversationLinkNode_1 {id=#3, class=ConversationLinkNode, properties {remoteEntryTag {tagName=Conversation.Entry.NPCLibrary.Quests}}}
]
